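; Music event actor for BrickSoupPastelGlaze that registers its audio components with the MusicEvent subsystem as gameplay events.
; A change-gated MusicSpawned setter toggles default-submix audio analysis and registers components, whose finish callback calls MusicSpawned(false) a tick later to work around loading-related stops.
; AudioGameplayVolume toggles drive a control bus mix and compressor preset settings; EndPlay switches mix and analysis off.
; A debug event prints the music amplitude material parameter.
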(event Custom|Music%20Spawned
  (params (MusicSpawned :type Boolean))
  (for _array_element (Actor|GetComponentsByClass "/Script/Engine.AudioComponent")
    (bind _output_get (Variables|Default|SetRegisteredAudioComponent _array_element))
    (Audio|AddEventbyComponent (GameInstanceSubsystems|GetMusicEventSubsystem) _output_get "(TagName=\"MusicEventSubsystem.Event.Gameplay\")" "(GameplayTags=)" 0 "")
    (EventDispatchers|BindEventtoOnAudioFinished (Variables|Default|GetRegisteredAudioComponent) (EventDispatchers|CreateEvent :!event Custom|OnAudioFinished))))

(event Custom|SetMusicSpawned
  (params (MusicSpawned :type Boolean))
  (bind _updated_variable (Utilities|FlowControl|DeltaGateWildcard (Variables|Event|GetMusicSpawned) MusicSpawned)
    (:Success
      (CallFunction|OnMusicPlaying _updated_variable)
      (if _updated_variable
        (for _array_element (Actor|GetComponentsByClass "/Script/Engine.AudioComponent")
          (bind _output_get (Variables|Default|SetRegisteredAudioComponent _array_element))
          (Audio|AddEventbyComponent (GameInstanceSubsystems|GetMusicEventSubsystem) _output_get "(TagName=\"MusicEventSubsystem.Event.Gameplay\")" "(GameplayTags=)" 0 "")
          (EventDispatchers|BindEventtoOnAudioFinished (Variables|Default|GetRegisteredAudioComponent) (EventDispatchers|CreateEvent :!event Custom|OnAudioFinished)))))))

(event Custom|OnAudioFinished
  ;[0.02576 0.04626 0.15 0.5] Fix for Audio Stopping/Starting if loading is off.
  (Utilities|FlowControl|DelayUntilNextTick)
  (CallFunction|MusicSpawned false))

(event Custom|DebugTick
  (if (Variables|Debug|GetDebugMPC)
    (bind _returnvalue (Rendering|Material|GetScalarParameterValue (Variables|Default|GetMaterialParameterCollection) (Utilities|Name|MakeLiteralName "Amplitude_Music") :!class /Script/Engine.KismetMaterialLibrary))
    (Development|PrintString (Utilities|String|ToString(Float) _returnvalue) true true "(R=0.000000,G=0.660000,B=1.000000,A=1.000000)" 2.0 "Amplitude")))

(event EventEndPlay
  (params (EndPlayReason :type EEndPlayReason%20Enum :default "Destroyed"))
  (CallFunction|OnMusicPlaying false)
  (CallFunction|OnMixEnabled false))

(event OnToggledOn(AudioGameplayVolume)
  (CallFunction|OnMixEnabled true))

(event OnToggledOff(AudioGameplayVolume)
  (CallFunction|OnMixEnabled false))

(fn OnMixEnabled
  (params (MixEnabled :type Boolean))
  (bind _compressor (Variables|Default|GetCompressor))
  (bind _controlbusmix (Variables|Default|GetControlBusMix))
  (if MixEnabled
    (Audio|Modulation|ActivateControlBusMix _controlbusmix)
    (else
      (Audio|Modulation|DeactivateControlBusMix _controlbusmix)))
  (bind _structout (Utilities|Struct|SetmembersinSubmixEffectDynamicsProcessorSettings (Variables|/Script/AudioMixer.SubmixEffectDynamicsProcessorPreset|GetSettings _compressor) MixEnabled))
  (Audio|Effects|SetSettings _compressor _structout :!class /Script/AudioMixer.SubmixEffectDynamicsProcessorPreset))

(fn OnMusicPlaying
  (params (MusicPlaying :type Boolean))
  (bind _returnvalue (WorldSubsystems|GetFortAudioAnalysisSubsystem))
  (if MusicPlaying
    (AudioAnalysis|RegisterDefaultSubmix _returnvalue)
    (else
      (AudioAnalysis|UnRegisterDefaultSubmix _returnvalue))))

(fn ConstructionScript)
